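# Niagara effect for deploying the BlockStack glider, built from mesh cubes and a splash, both positioned via a socket reader.
# Cubes combines an age-curved spawn stream, stepping from brown through green to cyan, with four later staggered bursts.
# Spawn group selects each cube's socket, mesh, color, scale and drift.
# Splash fires two particles with the first burst, rendered as sprite and mesh with time-of-day dimmed blue tints.
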NiagaraSystem /BRCosmetics/Gadgets/Assets/VinderTech_GliderChute/Glider_BlockStack/FX/NS_Glider_BlockStack_Deploy
dependencies {
 usedDataInterfaces [&NiagaraDataInterfaceCurve, &NiagaraDataInterfaceColorCurve, &NiagaraDataInterfaceSocketReader]
 usedDynamicInputs [&FN_FrontEndFloatModifier]
 usedModules [
  &self:SystemSpawnScript.SystemScriptSource.SystemScriptGraph.NiagaraNodeAssignment_4.SetVariables_#1
  &SystemState
  &self:SystemSpawnScript.SystemScriptSource.SystemScriptGraph.NiagaraNodeAssignment_3.SetVariables_#2
 ]
}
emitters {
 Cubes {
  emitterUpdateScript [
   &EmitterState(Life Cycle Mode=::NewEnumerator1, Loop Behavior=::NewEnumerator1, UseLoopDelay=false)
   &SpawnBurst_Instantaneous(Spawn Count=14, Spawn Time=1.7, Use Spawn Probability=false, UseLoopCountLimit=false)
   &SpawnBurst_Instantaneous(Spawn Count=5, Spawn Group=1, Spawn Time=1.82, Use Spawn Probability=false, UseLoopCountLimit=false) {module=SpawnBurst_Instantaneous001}
   &SpawnBurst_Instantaneous(Spawn Count=5, Spawn Group=2, Spawn Time=1.9, Use Spawn Probability=false, UseLoopCountLimit=false) {module=SpawnBurst_Instantaneous002}
   &SpawnBurst_Instantaneous(Spawn Count=5, Spawn Group=3, Spawn Time=1.8, Use Spawn Probability=false, UseLoopCountLimit=false) {module=SpawnBurst_Instantaneous003}
   &SpawnRate(
    SpawnGroup=4
    SpawnRate=&FloatFromCurve(
     FloatCurve=DataInterface() {propertyValues {Curve {keys [
         {arriveTangent=5.5617, interpMode=RCIM_Cubic, leaveTangent=5.5617, tangentMode=RCTM_User, time=0.65, value=0}
         <at 0.9 1 RCIM_Cubic>
         {arriveTangent=-7.5582, tangentMode=RCTM_User, time=1.3608, value=0}
        ]}}}
     CurveIndex=$Engine.System.Age
     Scale Curve=50
    )
    Use Spawn Probability=true
   )
  ]
  particleSpawnScript [
   &InitializeParticle(
    Color=&MergeHelper_ColorSwitch(
     0=&RandomRangeLinearColor(Minimum=<rgba 0.0823 0.807 0.855 1>, Maximum=<rgba 0.2939 0.8201 0.855 1>)
     1=&RandomRangeLinearColor(Minimum=<rgba 0.6167 0.7245 0.786 1>, Maximum=<rgba 0.5517 0.6481 0.7031 1>)
     2=&RandomRangeLinearColor(Minimum=<rgba 0.6167 0.7245 0.786 1>, Maximum=<rgba 0.5517 0.6481 0.7031 1>)
     3=&RandomRangeLinearColor(Minimum=<rgba 0.6167 0.7245 0.786 1>, Maximum=<rgba 0.5517 0.6481 0.7031 1>)
     4=&RandomRangeLinearColor(
      Minimum=&ColorFromCurve(
       ColorCurve=DataInterface() {propertyValues {
         AlphaCurve {keys [<at 0 1 RCIM_Constant>,<at 1 1 RCIM_Constant>]}
         BlueCurve {keys [<at 0 0.0871 RCIM_Constant>,<at 1 0.1441 RCIM_Constant>,<at 1.2 0.855 RCIM_Constant>]}
         GreenCurve {keys [<at 0 0.1725 RCIM_Constant>,<at 1 0.6795 RCIM_Constant>,<at 1.2 0.807 RCIM_Constant>]}
         RedCurve {keys [<at 0 0.6042 RCIM_Constant>,<at 1 0.2918 RCIM_Constant>,<at 1.2 0.0823 RCIM_Constant>]}
        }}
       CurveIndex=$Engine.System.Age
      )
      Maximum=&ColorFromCurve(
       ColorCurve=DataInterface() {propertyValues {
         AlphaCurve {keys [<at 0 1 RCIM_Constant>,<at 1 1 RCIM_Constant>]}
         BlueCurve {keys [<at 0 0.0732 RCIM_Constant>,<at 1 0.1441 RCIM_Constant>,<at 1.2 0.855 RCIM_Constant>]}
         GreenCurve {keys [<at 0 0.0958 RCIM_Constant>,<at 1 0.6795 RCIM_Constant>,<at 1.2 0.807 RCIM_Constant>]}
         RedCurve {keys [<at 0 0.3125 RCIM_Constant>,<at 1 0.2918 RCIM_Constant>,<at 1.2 0.0823 RCIM_Constant>]}
        }}
       CurveIndex=$Engine.System.Age
      )
     )
     Index=&SpawnGroup()
     Num=::NewEnumerator3
    )
    Lifetime Max=0.9
    Lifetime Min=0.5
    Lifetime Mode=::NewEnumerator1
    Mesh Scale Mode=::NewEnumerator1
    Mesh Uniform Scale Max=1.25
    Mesh Uniform Scale Min=0.7
    Sprite Size Mode=::NewEnumerator3
    Uniform Sprite Size=8
    UsePositionOffset=true
    Write Material Random=true
   )
   &InitialMeshOrientation(Mesh Orientation Mode=::NewEnumerator2, Use Rotation Vector=false)
   &AddVelocity(
    Velocity=&RandomRangeVector(Minimum=<-60 -60 61.7782>, Maximum=<60 60 192.3832>)
    Velocity Speed Scale=1.4
   )
   /SetVariables(Particles.MeshIndex=&MergeHelper_IntSwitch(
     0=0
     1=&RandomRangeInt(Minimum=1, Maximum=2)
     2=&RandomRangeInt(Minimum=1, Maximum=2)
     3=&RandomRangeInt(Minimum=1, Maximum=2)
     4=0
     5=?
     Index=&SpawnGroup()
     Num=::NewEnumerator3
    )) {isSetParameters=true}
  ]
  particleUpdateScript [
   &ParticleState()
   &SocketLocation(
    SamplePosOffset=true
    Sampled Position Offset=&Add_Vector(
     A=&MergeHelper_VectorSwitch(
      0=&RandomRangeVector(Minimum=<-15 -30.2549 0>, Maximum=<20 30 0>)
      1=&RandomRangeVector(Minimum=<-10 -10 -10>, Maximum=<10 10 10>)
      2=&RandomRangeVector(Minimum=<-10 -10 -10>, Maximum=<10 10 10>)
      3=&RandomRangeVector(Minimum=<-10 -10 -10>, Maximum=<10 10 10>)
      4=&Add_Vector(
       A=&RandomRangeVector(Minimum=<-50 -50 -20>, Maximum=<50 50 -100>)
       B=&MakeVector(
        X=0
        Y=0
        Z=&FloatFromCurve(FloatCurve=DataInterface() {propertyValues {Curve {keys [<at 0.65 0>,<at 1.3 1>]}}}, CurveIndex=$Engine.System.Age, Scale Curve=40)
       )
      )
      Index=&SpawnGroup()
      Num=::NewEnumerator3
     )
     B=&MakeVector(
      X=&FloatFromCurve(
       FloatCurve=DataInterface() {propertyValues {Curve {keys [{arriveTangent=1.3697, interpMode=RCIM_Cubic, leaveTangent=1.3697, tangentMode=RCTM_User, time=0, value=0}, <at 1 1>]}}}
       Scale Curve=&Multiply_Float(
        A=&RandomRangeFloat(Minimum=-70, Maximum=70)
        B=&MergeHelper_FloatSwitch(0=1, 1=1.5, 2=1.5, 3=1.5, 4=3, Index=&SpawnGroup(), Num=::NewEnumerator3)
       )
      )
      Y=&FloatFromCurve(
       FloatCurve=DataInterface() {propertyValues {Curve {keys [{arriveTangent=1.3697, interpMode=RCIM_Cubic, leaveTangent=1.3697, tangentMode=RCTM_User, time=0, value=0}, <at 1 1>]}}}
       Scale Curve=&Multiply_Float(
        A=&RandomRangeFloat(Minimum=-70, Maximum=70)
        B=&MergeHelper_FloatSwitch(0=1, 1=1.5, 2=1.5, 3=1.5, 4=3, Index=&SpawnGroup(), Num=::NewEnumerator3)
       )
      )
      Z=&FloatFromCurve(
       FloatCurve=DataInterface() {propertyValues {Curve {keys [
           {arriveTangent=1.3697, interpMode=RCIM_Cubic, leaveTangent=1.3697, tangentMode=RCTM_User, time=0, value=0}
           {arriveTangent=-1, interpMode=RCIM_Cubic, leaveTangent=-1, time=0.4748, value=1}
           <at 1 -1>
          ]}}}
       Scale Curve=&RandomRangeFloat(Minimum=10, Maximum=80)
      )
     )
    )
    Socket Index=&MergeHelper_IntSwitch(0=0, 1=1, 2=2, 3=3, 4=4, 5=4, Index=&SpawnGroup(), Num=::NewEnumerator4)
    Socket Reader=$System.Socket Reader
    Socket Sampling Mode=::NewEnumerator3
    VelocitySpeedLimit=false
   )
   &ScaleMeshSize(Scale Factor=&VectorFromFloat(Value=&FloatFromCurve(
      FloatCurve=DataInterface() {propertyValues {Curve {keys [<at 0 0 RCIM_Cubic>,<at 0.2809 1 RCIM_Cubic>,<at 1 0 RCIM_Cubic>]}}}
      Scale Curve=0.8
     ))) {module=ScaleMeshSize002}
   &ScaleMeshSize(Scale Factor=&MergeHelper_VectorSwitch(
     0=&RandomRangeVector(Minimum=<2 2 2>, Maximum=<4 4 4>)
     1=&RandomRangeVector(Minimum=<1 1 1>, Maximum=<2 2 2>)
     2=&RandomRangeVector(Minimum=<1 1 1>, Maximum=<2 2 2>)
     3=&RandomRangeVector(Minimum=<1 1 1>, Maximum=<2 2 2>)
     4=&RandomRangeVector(Minimum=<4 4 4>, Maximum=<6 6 6>)
     Index=&SpawnGroup()
     Num=::NewEnumerator3
    )) {module=ScaleMeshSize001}
   &DynamicMaterialParameters(
    Index 0 Param 2=1
    Index 0 Param 3=1
    Module.Index0Float0Write=true
    Module.Index0Float1Write=true
    Module.Index0Float2Write=true
    Module.Index0Float3Write=true
   )
   &ScaleColor(
    Scale Alpha=&FloatFromCurve(FloatCurve=DataInterface() {propertyValues {Curve {keys [{interpMode=RCIM_Cubic, tangentMode=RCTM_Break, time=0.3, value=1}, <at 1 0>]}}})
    ScaleA=true
    ScaleRGB=true
   )
   &SolveForcesAndVelocity(Clamp Velocity=false, Limit Acceleration=false)
  ]
  rendererClasses [&NiagaraMeshRendererProperties]
  renderers [{rendererClass=&NiagaraMeshRendererProperties, rendererIndex=0}]
 }
 Splash {
  emitterUpdateScript [
   &EmitterState(Life Cycle Mode=::NewEnumerator1, Loop Behavior=::NewEnumerator1, Loop Delay=0.1, UseLoopDelay=false)
   &SpawnBurst_Instantaneous(Spawn Count=1, Spawn Time=1.7, Use Spawn Probability=false)
   &SpawnBurst_Instantaneous(Spawn Count=1, Spawn Group=1, Spawn Time=1.7, Use Spawn Probability=false) {module=SpawnBurst_Instantaneous004}
  ]
  particleSpawnScript [
   /SetVariables(Particles.VisibilityTag=&SpawnGroup()) {isSetParameters=true}
   &InitializeParticle(
    Lifetime=&MergeHelper_FloatSwitch(0=0.3, 1=0.4, 2=?, 3=?, Index=&SpawnGroup(), Num=::NewEnumerator0)
    Sprite Rotation Mode=::NewEnumerator2
    Sprite Size Mode=::NewEnumerator3
    Uniform Sprite Size=&Multiply_Float(A=180, B=1)
    UsePositionOffset=false
    Write Material Random=true
   )
   &DynamicMaterialParameters(
    Index 0 Param 1=&RandomRangeFloat(Minimum=0.4, Maximum=0.6)
    Index 0 Param 2=1
    Index 0 Param 3=1
    Index 0 Param 4=1
    Module.Param0WriteEnabled=true
    Module.Param1WriteEnabled=true
    Module.Param2WriteEnabled=true
    Module.Param3WriteEnabled=true
   )
   &CameraOffset(Camera Offset Amount=10)
  ]
  particleUpdateScript [
   &ParticleState()
   &SocketLocation(
    SamplePosOffset=true
    Sampled Position Offset=<4 0 16>
    Socket Reader=$System.Socket Reader
    Socket Sampling Mode=::NewEnumerator3
    VelocitySpeedLimit=false
   )
   &FN_Sprite_Attributes(
    Color=&MultiplyLinearColor(
     Linear Color=&Multiply_LinearColorByFloat(
      Linear Color=&ColorFromCurve(ColorCurve=DataInterface() {propertyValues {
         AlphaCurve {keys [<at 0 1>]}
         BlueCurve {keys [<at 0.234 1>,<at 0.7319 0.9844>]}
         GreenCurve {keys [<at 0.234 0.7267>,<at 0.7319 0.6116>]}
         RedCurve {keys [<at 0.234 0.2088>,<at 0.7319 0.209>]}
        }})
      Float=&FNTimeOfDayScalar(Night Scale=0.65, Morning Scale=0.9, Evening Scale=0.8)
     )
     Scale Factor=&MakeVector4FromVectorAndFloat(
      "Vector (RGB)"=<1 1 1>
      "Float (Alpha)"=&FloatFromCurve(FloatCurve=DataInterface() {propertyValues {Curve {keys [
           {arriveTangent=0.1189, interpMode=RCIM_Cubic, leaveTangent=0.1189, tangentMode=RCTM_User, time=0, value=1}
           {arriveTangent=-1.3216, interpMode=RCIM_Cubic, leaveTangent=-1.3216, tangentMode=RCTM_User, time=1, value=0}
          ]}}})
     )
    )
    Create Color Attribute=true
    Create Dynamic Material Parameter Attributes=true
    Create SpriteSize Attribute=true
    Index 0 Param 1=0.65
    Index 0 Param 2=1
    Index 0 Param 3=1
    Index 0 Param 4=10
    Module Output:ENiagara_AttributeSamplingApplyOutput=::NewEnumerator1
    Module.Index0Float0Write=true
    Module.Index0Float1Write=true
    Module.Index0Float2Write=true
    Module.Index0Float3Write=true
    SpriteSize=&Vector2DFromFloat(Value=&Multiply_Float(
      A=&Multiply_Float(
       A=160
       B=&Multiply_Float(
        A=&FloatFromCurve(FloatCurve=DataInterface() {propertyValues {Curve {keys [
             {arriveTangent=2.185, interpMode=RCIM_Cubic, leaveTangent=2.185, tangentMode=RCTM_User, time=0, value=0}
             {interpMode=RCIM_Cubic, tangentMode=RCTM_User, time=1, value=1}
            ]}}})
        B=1
       )
      )
      B=1
     ))
   ) {module=FN_Sprite_Attributes002}
   &FN_Mesh_Attributes(
    Color=&Multiply_LinearColorByFloat(
     Linear Color=&ColorFromCurve(ColorCurve=DataInterface() {propertyValues {
        AlphaCurve {keys [<at 0 1 RCIM_Cubic>,<at 1 0 RCIM_Cubic>]}
        BlueCurve {keys [<at 0.0287 1>, {arriveTangent=-0.2458, interpMode=RCIM_Cubic, leaveTangent=-0.2458, time=0.5636, value=1}, <at 0.8777 0.7913>]}
        GreenCurve {keys [<at 0.0287 0.8756>, {arriveTangent=-0.2392, interpMode=RCIM_Cubic, leaveTangent=-0.2392, time=0.5636, value=0.7872}, <at 0.8777 0.6724>]}
        RedCurve {keys [<at 0.0287 0.526>, {arriveTangent=-0.6074, interpMode=RCIM_Cubic, leaveTangent=-0.6074, time=0.5636, value=0.1896}, <at 0.8777 0.0103>]}
       }})
     Float=&FNTimeOfDayScalar(Night Scale=0.65, Morning Scale=0.9, Evening Scale=0.8)
    )
    Create Color Attribute=true
    Create Dynamic Parameter Attribute=true
    Create Orientation Attribute=true
    Create Scale Attribute=true
    Index 0 Param 2:NiagaraFloat=&FloatFromCurve(FloatCurve=DataInterface() {propertyValues {Curve {
        defaultValue=1
        keys [
         {arriveTangent=1.4575, interpMode=RCIM_Cubic, leaveTangent=1.4575, tangentMode=RCTM_User, time=0, value=0}
         {arriveTangent=0.5509, interpMode=RCIM_Cubic, leaveTangent=0.5509, tangentMode=RCTM_User, time=1, value=1}
        ]
       }}})
    Module Output=::NewEnumerator1
    Module.Index0Float0Write=true
    Module.Index0Float1Write=true
    Module.Index0Float2Write=true
    Module.Index0Float3Write=true
    Orientation:Quat4f=&MakeQuaternion(
     Quaternion From=0
     Axis=<0 1 0>
     Angle=0.25
     From=?
     To=?
     Stabilization Mode:ENiagara_OrientVectorCount=::NewEnumerator0
     Side Direction:Vector3f=?
     Up Direction:Vector3f=?
     Angle Type=::NewEnumerator1
     Normalize:NiagaraBool=true
    )
    Scale=&VectorFromFloat(Value=&Multiply_Float(
      A=&Lerp_Float(
       A=0.6
       B=0.6
       Alpha=&FloatFromCurve(FloatCurve=DataInterface() {propertyValues {Curve {keys [
            {arriveTangent=2.0628, interpMode=RCIM_Cubic, leaveTangent=2.0628, tangentMode=RCTM_User, time=0, value=0}
            {arriveTangent=0.2586, interpMode=RCIM_Cubic, leaveTangent=0.2586, tangentMode=RCTM_User, time=1, value=1}
           ]}}})
      )
      B=1
     ))
   ) {module=FN_Mesh_Attributes001}
   /Game/Sprite/FN_SpriteAlignmentAndFacing(SpriteAlignment:Vector3f=<1 0 0>, SpriteFacing:Vector3f=<0 0 1>)
  ]
  rendererClasses [&NiagaraSpriteRendererProperties, &NiagaraMeshRendererProperties]
  renderers [
   {rendererClass=&NiagaraSpriteRendererProperties}
   {rendererClass=&NiagaraMeshRendererProperties, rendererIndex=1}
  ]
 }
}
summary {
 systemName=NS_Glider_BlockStack_Deploy
 emitters [
  {emitterName=Cubes, rendererClasses [&NiagaraMeshRendererProperties]}
  {emitterName=Splash, rendererClasses [&NiagaraSpriteRendererProperties, &NiagaraMeshRendererProperties]}
 ]
 userVariables [
  User.bIsFrontEnd {type=NiagaraBool}
  User.bIsFrontEndPreview {type=NiagaraBool}
  User.bIsFullyDeployed {type=NiagaraBool}
 ]
}
