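# Niagara trail effect for the BlockStack glider, built from three mesh emitters: clouds, tinted cubes and surfing wind lines.
# Clouds and Cubes use spawn groups to pick glider sockets, offsets and scales, then stream along +Y with curl noise, scaling and fading out.
# Wind lines spawn on a torus, shift with User.RightDot and dim in the front end.
# A system Timeline value drives every emitter's spawn probability.
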NiagaraSystem /BRCosmetics/Gadgets/Assets/VinderTech_GliderChute/Glider_BlockStack/FX/NS_Glider_BlockStack_Trail
dependencies {
 usedDataInterfaces [&NiagaraDataInterfaceCurve, &NiagaraDataInterfaceColorCurve, &NiagaraDataInterfaceSocketReader]
 usedDynamicInputs [
  &SetBoolByFloatComparison
  &ClampFloat
  &FNTimeOfDayScalar
 ]
 usedModules [
  &self:SystemSpawnScript.SystemScriptSource.SystemScriptGraph.NiagaraNodeAssignment_0.SetVariables_#1
  &self:SystemSpawnScript.SystemScriptSource.SystemScriptGraph.NiagaraNodeAssignment_4.SetVariables_#2
  &SystemState
  &Timeline
  &InterpolateOverTime
  &self:SystemSpawnScript.SystemScriptSource.SystemScriptGraph.NiagaraNodeAssignment_3.SetVariables_#3
 ]
}
emitters {
 Clouds {
  emitterUpdateScript [
   &EmitterState(Life Cycle Mode=::NewEnumerator1, Loop Behavior=::NewEnumerator1, Loop Duration Mode=::NewEnumerator1, UseLoopDelay=false)
   &SpawnRate(
    Spawn Probability=&Multiply_Float(A=1, B=$System.Timeline.Time)
    SpawnRate=36.8122
    Use Spawn Probability=true
   )
  ]
  particleSpawnScript [
   &FN_SetSpawnGroup(Number of Spawn Groups=3)
   &InitializeParticle(
    Color=<rgba 1 1 1 2>
    Lifetime Max=0.4
    Lifetime Min=0.3
    Lifetime Mode=::NewEnumerator1
    Mesh Scale Mode=::NewEnumerator1
    Mesh Uniform Scale Max=1.25
    Mesh Uniform Scale Min=0.7
    Sprite Size Mode=::NewEnumerator3
    Uniform Sprite Size=8
    UsePositionOffset=true
    Write Material Random=true
   )
   &InitialMeshOrientation(
    Mesh Orientation Mode=::NewEnumerator2
    Rotation=&RandomRangeVector(
     Minimum=<0 0 0>
     Maximum=&MakeVector(
      X=0
      Y=0
      Z=&MergeHelper_FloatSwitch(0=1, 1=1, 2=1, 3=?, Index=&SpawnGroup(), Num=::NewEnumerator1)
     )
    )
    Use Rotation Vector=true
   )
   &SocketLocation(
    SamplePosOffset=true
    Sampled Position Offset=&Subtract_Vector(
     A=&MergeHelper_VectorSwitch(
      0=&RandomRangeVector(Minimum=<-10 -10 -10>, Maximum=<10 10 -3>)
      1=&RandomRangeVector(Minimum=<-10 -10 -10>, Maximum=<10 10 -3>)
      2=&RandomRangeVector(Minimum=<-10 -10 -10>, Maximum=<10 10 -3>)
      3=&RandomRangeVector(Minimum=<-10 -10 -10>, Maximum=<10 10 -3>)
      4=&RandomRangeVector(Minimum=<-10 -10 -10>, Maximum=<10 10 -3>)
      5=&RandomRangeVector(Minimum=<-10 -10 -10>, Maximum=<10 10 -3>)
      6=&RandomRangeVector(Minimum=<-16 0 -3>, Maximum=<15 0 3>)
      7=&RandomRangeVector(Minimum=<-16 0 -3>, Maximum=<22 0 3>)
      Index=&SpawnGroup()
      Num=::NewEnumerator1
     )
     B=<0 0 0>
    )
    Socket Index=&MergeHelper_IntSwitch(0=3, 1=4, 2=5, 3=?, 4=?, 5=?, Index=&SpawnGroup(), Num=::NewEnumerator1)
    Socket Reader=$System.Socket Reader
    Socket Sampling Mode=::NewEnumerator3
    VelocitySpeedLimit=false
   )
   &AddVelocity(Velocity=<0 500 3>, Velocity Speed Scale=&RandomRangeFloat(Minimum=0.7, Maximum=1))
   /SetVariables(Particles.MeshIndex=&MergeHelper_IntSwitch(
     0=&RandomRangeInt(Minimum=0, Maximum=1)
     1=&RandomRangeInt(Minimum=0, Maximum=1)
     2=&RandomRangeInt(Minimum=0, Maximum=1)
     3=&RandomRangeInt(Minimum=1, Maximum=2)
     4=&RandomRangeInt(Minimum=1, Maximum=2)
     5=&RandomRangeInt(Minimum=1, Maximum=2)
     Index=&SpawnGroup()
     Num=::NewEnumerator1
    )) {isSetParameters=true}
  ]
  particleUpdateScript [
   &ParticleState()
   &ScaleMeshSize(Scale Factor=&MakeVector(
     X=&FloatFromCurve(FloatCurve=DataInterface() {propertyValues {Curve {keys [
          <at 0 0 RCIM_Cubic>
          <at 0.1733 1>
          {arriveTangent=-1.2097, interpMode=RCIM_Cubic, leaveTangent=-1.2097, time=0.5, value=1}
          <at 1 0 RCIM_Cubic>
         ]}}})
     Y=&FloatFromCurve(FloatCurve=DataInterface() {propertyValues {Curve {keys [
          <at 0 0 RCIM_Cubic>
          <at 0.1733 1>
          {arriveTangent=-1.2097, interpMode=RCIM_Cubic, leaveTangent=-1.2097, time=0.5, value=1}
          <at 1 0 RCIM_Cubic>
         ]}}})
     Z=&FloatFromCurve(FloatCurve=DataInterface() {propertyValues {Curve {keys [
          <at 0 0 RCIM_Cubic>
          <at 0.1733 1>
          {arriveTangent=-1.2097, interpMode=RCIM_Cubic, leaveTangent=-1.2097, time=0.5, value=1}
          <at 1 0 RCIM_Cubic>
         ]}}})
    )) {module=ScaleMeshSize002}
   &ScaleMeshSize(Scale Factor=&MergeHelper_VectorSwitch(
     0=&RandomRangeVector(Minimum=<1 1 1>, Maximum=<1.5 1.5 1.5>)
     1=&RandomRangeVector(Minimum=<1 1 1>, Maximum=<1.5 1.5 1.5>)
     2=&RandomRangeVector(Minimum=<1 1 1>, Maximum=<1.5 1.5 1.5>)
     3=&RandomRangeVector(Minimum=<1 1 1>, Maximum=<1.5 1.5 1.5>)
     4=&RandomRangeVector(Minimum=<1 1 1>, Maximum=<1.5 1.5 1.5>)
     5=&RandomRangeVector(Minimum=<1 1 1>, Maximum=<1.5 1.5 1.5>)
     6=&RandomRangeVector(Minimum=<2.5 2 2>, Maximum=<3 2.5 2.5>)
     7=&RandomRangeVector(Minimum=<2.5 2 2>, Maximum=<3 2.5 2.5>)
     Index=&SpawnGroup()
     Num=::NewEnumerator1
    )) {module=ScaleMeshSize001}
   &DynamicMaterialParameters(
    Index 0 Param 2=1
    Index 0 Param 3=1
    Module.Index0Float0Write=true
    Module.Index0Float1Write=true
    Module.Index0Float2Write=true
    Module.Index0Float3Write=true
   )
   &CurlNoiseForce(
    Bias Noise=false
    Noise Frequency=55
    Noise Quality / Cost=::NewEnumerator0
    Noise Strength=&RandomRangeFloat(Minimum=10, Maximum=20)
    Pan Noise=false
   )
   &ScaleSpriteSize(
    Scale Sprite Size Mode=::NewEnumerator0
    Uniform Scale Factor=&MakeFloatFromVector(Vector=$Particles.Scale, Channel=::NewEnumerator0)
   )
   &ScaleColor(
    Scale Alpha=&FloatFromCurve(FloatCurve=DataInterface() {propertyValues {Curve {keys [{interpMode=RCIM_Cubic, tangentMode=RCTM_Break, time=0.3, value=1}, <at 1 0>]}}})
    Scale RGB=&Multiply_VectorByFloat(Vector=&MakeVectorFromLinearColorRGB(Color=<rgba 0.6167 0.7245 0.786 1>), Float=1)
    ScaleA=true
    ScaleRGB=true
   )
   &SolveForcesAndVelocity(Clamp Velocity=false, Limit Acceleration=false)
  ]
  rendererClasses [&NiagaraMeshRendererProperties]
  renderers [{rendererClass=&NiagaraMeshRendererProperties, rendererIndex=0}]
 }
 Cubes {
  emitterUpdateScript [
   &EmitterState(Life Cycle Mode=::NewEnumerator1, Loop Behavior=::NewEnumerator1, Loop Duration Mode=::NewEnumerator1, UseLoopDelay=false)
   &SpawnRate(
    Spawn Probability=&Multiply_Float(A=0.4, B=$System.Timeline.Time)
    SpawnRate=29.7644
    Use Spawn Probability=true
   )
  ]
  particleSpawnScript [
   &FN_SetSpawnGroup(Number of Spawn Groups=6)
   &InitializeParticle(
    Color=<rgba 1 1 1 2>
    Lifetime Max=0.6
    Lifetime Min=0.4
    Lifetime Mode=::NewEnumerator1
    Mesh Scale Mode=::NewEnumerator1
    Mesh Uniform Scale Max=1.25
    Mesh Uniform Scale Min=0.7
    Sprite Size Mode=::NewEnumerator3
    Uniform Sprite Size=8
    UsePositionOffset=true
    Write Material Random=true
   )
   &InitialMeshOrientation(
    Mesh Orientation Mode=::NewEnumerator2
    Rotation=&RandomRangeVector(Minimum=<0 0 0>, Maximum=&MakeVector(X=0, Y=0, Z=0))
    Use Rotation Vector=true
   )
   &SocketLocation(
    SamplePosOffset=true
    Sampled Position Offset=&MergeHelper_VectorSwitch(
     0=<0 0 0>
     1=<0 0 0>
     2=<0 0 0>
     3=&RandomRangeVector(Minimum=<-16 0 -3>, Maximum=<15 0 3>)
     4=&RandomRangeVector(Minimum=<-16 0 -3>, Maximum=<22 0 3>)
     5=&RandomRangeVector(Minimum=<-5 -10 -20>, Maximum=<5 10 20>)
     6=&RandomRangeVector(Minimum=<-16 0 -3>, Maximum=<15 0 3>)
     7=&RandomRangeVector(Minimum=<-16 0 -3>, Maximum=<22 0 3>)
     Index=&SpawnGroup()
     Num=::NewEnumerator4
    )
    Socket Index=&MergeHelper_IntSwitch(0=0, 1=1, 2=2, 3=6, 4=7, 5=8, Index=&SpawnGroup(), Num=::NewEnumerator4)
    Socket Reader=$System.Socket Reader
    Socket Sampling Mode=::NewEnumerator3
    VelocitySpeedLimit=false
   )
   &AddVelocity(
    Velocity=&MergeHelper_VectorSwitch(
     0=<0 700 20>
     1=<0 700 20>
     2=<0 700 20>
     3=<0 700 20>
     4=<0 700 20>
     5=&RandomRangeVector(Minimum=<0 700 20>, Maximum=<0 700 20>)
     Index=&SpawnGroup()
     Num=::NewEnumerator4
    )
    Velocity Speed Scale=&RandomRangeFloat(Minimum=0.7, Maximum=1)
   )
  ]
  particleUpdateScript [
   &ParticleState()
   &ScaleMeshSize(Scale Factor=&MakeVector(
     X=&FloatFromCurve(
      FloatCurve=DataInterface() {propertyValues {Curve {keys [
          <at 0 0 RCIM_Cubic>
          <at 0.08 1>
          {arriveTangent=-1.087, interpMode=RCIM_Cubic, leaveTangent=-1.087, time=0.662, value=1}
          <at 1 0 RCIM_Cubic>
         ]}}}
      Scale Curve=1.1
     )
     Y=&FloatFromCurve(FloatCurve=DataInterface() {propertyValues {Curve {keys [
          <at 0 0 RCIM_Cubic>
          <at 0.08 1>
          {arriveTangent=-1.087, interpMode=RCIM_Cubic, leaveTangent=-1.087, time=0.662, value=1}
          <at 1 0 RCIM_Cubic>
         ]}}})
     Z=&FloatFromCurve(FloatCurve=DataInterface() {propertyValues {Curve {keys [
          <at 0 0 RCIM_Cubic>
          <at 0.08 1>
          {arriveTangent=-1.087, interpMode=RCIM_Cubic, leaveTangent=-1.087, time=0.662, value=1}
          <at 1 0 RCIM_Cubic>
         ]}}})
    )) {module=ScaleMeshSize002}
   &ScaleMeshSize(Scale Factor=&MergeHelper_VectorSwitch(
     0=&RandomRangeVector(Minimum=<3 1.2 3>, Maximum=<5 1.2 3.7>)
     1=&RandomRangeVector(Minimum=<5 1.2 3.2>, Maximum=<6 1.2 3.8>)
     2=&RandomRangeVector(Minimum=<3 1.2 2.2>, Maximum=<4 1.2 2.8>)
     3=&RandomRangeVector(Minimum=<2.5 2 2>, Maximum=<3 2.5 2.5>)
     4=&RandomRangeVector(Minimum=<2.5 2 2>, Maximum=<3 2.5 2.5>)
     5=&RandomRangeVector(Minimum=<3 1.2 3>, Maximum=<5 1.2 3.7>)
     6=&RandomRangeVector(Minimum=<2.5 2 2>, Maximum=<3 2.5 2.5>)
     7=&RandomRangeVector(Minimum=<2.5 2 2>, Maximum=<3 2.5 2.5>)
     Index=&SpawnGroup()
     Num=::NewEnumerator4
    )) {module=ScaleMeshSize001}
   &DynamicMaterialParameters(
    Index 0 Param 2=1
    Index 0 Param 3=1
    Module.Index0Float0Write=true
    Module.Index0Float1Write=true
    Module.Index0Float2Write=true
    Module.Index0Float3Write=true
   )
   &CurlNoiseForce(
    Bias Noise=false
    Noise Frequency=55
    Noise Quality / Cost=::NewEnumerator0
    Noise Strength=&RandomRangeFloat(Minimum=10, Maximum=20)
    Pan Noise=false
   )
   &ScaleSpriteSize(
    Scale Sprite Size Mode=::NewEnumerator0
    Uniform Scale Factor=&MakeFloatFromVector(Vector=$Particles.Scale, Channel=::NewEnumerator0)
   )
   &ScaleColor(
    Scale Alpha=&FloatFromCurve(FloatCurve=DataInterface() {propertyValues {Curve {keys [{interpMode=RCIM_Cubic, tangentMode=RCTM_Break, time=0.3, value=1}, <at 1 0>]}}})
    Scale RGB=&Multiply_VectorByFloat(
     Vector=&MakeVectorFromLinearColorRGB(Color=&MergeHelper_ColorSwitch(
       0=<rgba 0.0823 0.807 0.855 1>
       1=<rgba 0.0823 0.807 0.855 1>
       2=<rgba 0.0823 0.807 0.855 1>
       3=<rgba 0.2918 0.6795 0.1441 1>
       4=<rgba 0.2918 0.6795 0.1441 1>
       5=&ColorFromCurve(ColorCurve=DataInterface() {propertyValues {
          AlphaCurve {keys [<at 0 1 RCIM_Cubic>,<at 1 1 RCIM_Cubic>]}
          BlueCurve {keys [<at 0 0.7605 RCIM_Cubic>,<at 1 0.855 RCIM_Cubic>]}
          GreenCurve {keys [<at 0 0.6308 RCIM_Cubic>,<at 1 0.807 RCIM_Cubic>]}
          RedCurve {keys [<at 0 0.0009 RCIM_Cubic>,<at 1 0.0823 RCIM_Cubic>]}
         }})
       Index=&SpawnGroup()
       Num=::NewEnumerator4
      ))
     Float=1.0525
    )
    ScaleA=true
    ScaleRGB=true
   )
   &SolveForcesAndVelocity(Clamp Velocity=false, Limit Acceleration=false)
  ]
  rendererClasses [&NiagaraMeshRendererProperties]
  renderers [{rendererClass=&NiagaraMeshRendererProperties, rendererIndex=0}]
 }
 WindLines_-_SURFING {
  emitterUpdateScript [
   &EmitterState()
   &SpawnRate(Spawn Probability=$System.Timeline.Time, SpawnGroup=1, SpawnRate=100, Use Spawn Probability=true) {module=SpawnRate001}
  ]
  particleSpawnScript [
   &InitializeParticle(Mesh Scale Max=<25 1 1>, Mesh Scale Min=<10 0.5 0.5>, Mesh Scale Mode=::NewEnumerator2, UsePositionOffset=false, Write Material Random=true)
   &InitialMeshOrientation(
    Mesh Orientation Mode=::NewEnumerator2
    Rotation=&RandomRangeVector(Minimum=<0 0 0>, Maximum=<1 0 0>)
    Use Rotation Vector=true
   )
   &TorusLocation(
    Align Torus With Axis=true
    Handle Radius=0
    Large Radius:NiagaraFloat=&RandomRangeFloat(Minimum=120, Maximum=240)
    MaskSpawnGroup=false
    Module.Override Randomness=false
    Offset=<0 -180 0>
    Spawn Group Mask=1
    Torus Axis=<0 1 0>
    Torus Origin:NiagaraPosition=&Add_Vector(A=<0 -500 0>, B=&Lerp_Vector(A=<0 0 0>, B=<200 0 0>, Alpha=$User.RightDot))
   ) {module=TorusLocation001}
   &AddVelocity(
    Rotation Coordinate Space=::Simulation
    Velocity=&Multiply_Vector(
     A=&RandomRangeVector(Minimum=<0 -7 0>, Maximum=<0 -5 0>)
     B=&MakeVector(X=1, Y=&MakeCustomFloatFromBool(Bool=$User.bIsFrontEndPreview, True Float=1, False Float=1), Z=1)
    )
    Velocity Speed Scale=-250
   ) {module=AddVelocity002}
   &AddVelocity(
    Module.Use Falloff Distance=false
    Origin Offset=&AddVectorToPosition(Position=<0 0 0>, Vector=&Lerp_Vector(A=<0 0 0>, B=<500 0 0>, Alpha=$User.RightDot))
    Override Seed=false
    Velocity Mode=::NewEnumerator1
    Velocity Speed=500
   ) {module=AddVelocity003}
  ]
  particleUpdateScript [
   &ScaleVelocity(Velocity Scale=&VectorFromFloat(Value=&FloatFromCurve(
      FloatCurve=DataInterface() {propertyValues {Curve {keys [
          {interpMode=RCIM_Cubic, tangentMode=RCTM_User, time=0, value=0}
          {arriveTangent=2, interpMode=RCIM_Cubic, leaveTangent=2, time=0.1, value=0.975}
          {interpMode=RCIM_Cubic, tangentMode=RCTM_User, time=0.5, value=1}
          {arriveTangent=-2, interpMode=RCIM_Cubic, leaveTangent=-2, time=0.9, value=1}
          {tangentMode=RCTM_User, time=1, value=0}
         ]}}}
      CurveIndex=&RemapRange(Input Value To Remap=$User.RightDot, Input Min=-1, Input Max=1, Output Min=0, Output Max=1)
     )))
   &SolveForcesAndVelocity(Clamp Velocity=false, Limit Acceleration=false)
   &ParticleState()
   &ScaleMeshSize(Scale Factor=&VectorFromFloat(Value=&FloatFromCurve(FloatCurve=DataInterface() {propertyValues {Curve {keys [
          {tangentMode=RCTM_User, time=0, value=1}
          {arriveTangent=-0.0405, interpMode=RCIM_Cubic, leaveTangent=-0.0405, tangentMode=RCTM_User, time=0.7009, value=1}
          {arriveTangent=-8.0089, interpMode=RCIM_Cubic, leaveTangent=-8.0089, tangentMode=RCTM_User, time=1, value=0}
         ]}}})))
   &DynamicMaterialParameters(
    Index 0 Param 2=2
    Index 0 Param 3=1
    Index 0 Param 4=1
    Module.Index0Float0Write=true
    Module.Index0Float1Write=true
    Module.Index0Float2Write=true
    Module.Index0Float3Write=true
   )
   &ScaleColor(
    Scale Alpha=&FloatFromCurve(
     FloatCurve=DataInterface() {propertyValues {Curve {keys [
         {interpMode=RCIM_Cubic, tangentMode=RCTM_User, time=0, value=0}
         {interpMode=RCIM_Cubic, tangentMode=RCTM_User, time=0.5, value=1.5}
         {tangentMode=RCTM_User, time=1, value=0}
        ]}}}
     Scale Curve=&Lerp_Float(A=0, B=&FN_FrontEndFloatModifier(Gameplay=1, Front End=0.3), Alpha=$System.Timeline.Time)
    )
    Scale RGB=&Multiply_VectorByFloat(Vector=<1 1 1>, Float=3)
    ScaleA=true
    ScaleRGB=true
   )
  ]
  rendererClasses [&NiagaraMeshRendererProperties]
  renderers [{rendererClass=&NiagaraMeshRendererProperties, rendererIndex=0}]
 }
}
summary {
 systemName=NS_Glider_BlockStack_Trail
 emitters [
  {emitterName=Cubes, rendererClasses [&NiagaraMeshRendererProperties]}
  {emitterName=WindLines_-_SURFING, rendererClasses [&NiagaraMeshRendererProperties]}
  {emitterName=Clouds, rendererClasses [&NiagaraMeshRendererProperties]}
 ]
 userVariables [
  User.bIsFrontEnd {type=NiagaraBool}
  User.bIsFrontEndPreview {type=NiagaraBool}
  User.bIsFullyDeployed {type=NiagaraBool}
  User.ForwardDot {type=NiagaraFloat, defaultValue=0, description=""}
  User.RightDot {type=NiagaraFloat, defaultValue=0, description=""}
 ]
}
