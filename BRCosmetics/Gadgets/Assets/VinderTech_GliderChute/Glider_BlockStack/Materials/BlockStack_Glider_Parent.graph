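# Parent material for the BlockStack glider, layering animated water and noise-displaced clouds between the CharacterParent 2023 PreFX and PostFX functions.
# A channel-selectable mask defines water, which combines noise-distorted panning normals, flattened by Fresnel, with tinted emissive ripples weighted by the pre-skinned normal's Z.
# Clouds push World Position Offset with panning noise, with an inverted 3D box mask defining their region.
# Pre-skinned local position drives projection and masking.
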Material /BRCosmetics/Gadgets/Assets/VinderTech_GliderChute/Glider_BlockStack/Materials/BlockStack_Glider_Parent
graph {
 properties {
  bAutomaticallySetUsageInEditor=false
  bHasPixelAnimation=true
  bUseMaterialAttributes=true
  bUsedWithMeshDeformer=true
  bUsedWithMorphTargets=true
  bUsedWithSkeletalMesh=true
  pixelDepthOffsetMode=PDOM_Legacy
  previewMesh=/BRCosmetics/Gadgets/Assets/VinderTech_GliderChute/Glider_BlockStack/Meshes/BlockStack_Glider
  shadingModel=MSM_FromMaterialExpression
  usedShadingModels=Default Lit | Subsurface | Clear Coat | Subsurface Profile | Cloth
 }
 nodes [
  n1 = Saturate(None=Add(A=NamedRerouteUsage(), B=ScalarParameter() {defaultValue=-0.15, group=BlockStack Glider|Water, parameterName=WaterRoughnessBias}))
  n2 = ComponentMask(None=PreSkinnedNormal()) {b=true}
  n3 = NamedRerouteUsage()
  n4 = TextureSample(UVs=Panner(Coordinate=Multiply(A=@n3) {constB=0.6}) {speedY=-0.1}) {samplerType=SAMPLERTYPE_LinearColor, texture=/BRCosmetics/Weapons/WeaponSkins/Wraps/Textures/RedSyko/T_Wrap_RedSyko}
  n5 = TextureSample(UVs=Panner(Coordinate=Multiply(A=@n3) {constB=0.6}) {speedX=0.1, speedY=0.1}) {samplerType=SAMPLERTYPE_Grayscale, texture=/BRCosmetics/Effects/Fort_Effects/Emotes/SwissKale/T_Blur_Noise}
  n6 = Multiply(A=VertexInterpolator(VS=Abs(None=@n2)).PS, B=Multiply(A=Multiply(A=@n4.G), B=LinearInterpolate(Alpha=@n5.R)))
  LinearInterpolate(A=@n1, Alpha=@n6) {constB=0}
  n7 = MaterialFunctionCall() {materialFunction=&MF_CharacterParent_2023_PreFX}
  MaterialFunctionCall(Material=NamedRerouteUsage(), M=@n7.M, Pre Skinned Local Position=@n7.M, Distance Blend=@n7.M, Fresnel=@n7.M) {materialFunction=&MF_CharacterParent_2023_PostFX}
  n8 = ScalarParameter() {defaultValue=0.25, group=BlockStack Glider|Water, parameterName=WaterNormalIntensity}
  Multiply(A=@n8) {constB=-1}
  NamedRerouteDeclaration(None=@n7.Result) {name=Material In, nodeColor=<rgba 0 0.2734 1 1>}
  n9 = LocalPosition() {localOrigin=InstancePreSkinning}
  n10 = ShaderStageSwitch(PixelShader=VertexInterpolator(VS=@n9.XYZ).PS, VertexShader=@n9.XYZ)
  n11 = Constant3Vector() {constant=<rgba 12 0 -50 0>}
  n12 = Constant3Vector() {constant=<rgba 175 500 200 0>}
  NamedRerouteDeclaration(None=OneMinus(None=MaterialFunctionCall!(A=@n10, B=@n11, Bounds=@n12, Edge Falloff) {materialFunction=/Engine/Functions/Engine_MaterialFunctions01/Gradient/BoxMask-3D}.Result)) {name=CloudsMask, nodeColor=<rgba 1 0.7266 0 1>}
  n13 = Multiply(A=LocalPosition() {localOrigin=InstancePreSkinning}.XY) {constB=0.01}
  NamedRerouteDeclaration(None=ShaderStageSwitch(PixelShader=VertexInterpolator(VS=@n13).PS, VertexShader=@n13)) {name=Pre-Skinned Position, nodeColor=<rgba 0 0.0156 1 1>}
  n14 = GetMaterialAttributes(None=NamedRerouteUsage()) {attributeGetTypes [
    #1
    #2
    #3
    #4
   ]}
  NamedRerouteDeclaration(None=@n14.Normal) {name=Normals In, nodeColor=<rgba 0 1 0.6719 1>}
  n15 = Add(A=NamedRerouteUsage(), B=Multiply(A=Subtract(A=@n5.R) {constB=0.5}) {constB=0.05})
  n16 = TextureSample(UVs=Panner(Coordinate=Multiply(A=@n15)) {speedX=0.01, speedY=-0.1}) {
   samplerType=SAMPLERTYPE_Normal
   texture=/BRCosmetics/Gadgets/Assets/VinderTech_GliderChute/Glider_BlockStack/Textures/T_BlockStack_Glider_Normal
  }
  n17 = Add(A=@n15) {constB=0.25}
  n18 = TextureSample(UVs=Panner(Coordinate=Multiply(A=@n17) {constB=0.6}) {speedX=-0.01, speedY=-0.025}) {
   samplerType=SAMPLERTYPE_Normal
   texture=&T_BlockStack_Glider_Normal
  }
  n19 = Constant() {r=0.5}
  n20 = MaterialFunctionCall(
   Normal=MaterialFunctionCall(
    BaseNormal=@n16.RGB
    AdditionalNormal=MaterialFunctionCall(Normal=@n18.RGB, Flatness=@n19) {materialFunction=&FlattenNormal}.Result
   ) {materialFunction=&BlendAngleCorrectedNormals}.Result
   Flatness=@n8
  ) {materialFunction=&FlattenNormal}
  n21 = Constant3Vector() {constant=<rgba 0 0 1 0>}
  n22 = Fresnel(Normal=VertexNormalWS()) {baseReflectFraction=0, exponent=3}
  NamedRerouteDeclaration(None=MaterialFunctionCall(BaseNormal=NamedRerouteUsage(), AdditionalNormal=LinearInterpolate(A=@n20.Result, B=@n21, Alpha=@n22)) {materialFunction=&BlendAngleCorrectedNormals}.Result) {name=Water Normal, nodeColor=<rgba 1 0.3516 0 1>}
  n23 = NamedRerouteUsage()
  n24 = NamedRerouteUsage()
  n25 = NamedRerouteUsage()
  NamedRerouteDeclaration(None=BlendMaterialAttributes(A=BlendMaterialAttributes(A=@n23, B=@n24, Alpha=@n25), B=NamedRerouteUsage(), Alpha=NamedRerouteUsage())) {name=Material Out, nodeColor=<rgba 1 0 0.9609 1>}
  NamedRerouteDeclaration(None=@n14.Emissive Color) {name=Emissive In, nodeColor=<rgba 0 1 0.7188 1>}
  NamedRerouteDeclaration(None=@n7.Tech Art Mask) {name=TechArtMask In, nodeColor=<rgba 0.8047 1 0 1>}
  n26 = ChannelMaskParameter(None=NamedRerouteUsage()) {group=BlockStack Glider|Water, maskChannel=Blue, parameterName=WaterMsk, defaultValue=<rgba 0 0 1 0>}
  NamedRerouteDeclaration(None=@n26.R) {name=WaterMask, nodeColor=<rgba 0.875 0 1 1>}
  n27 = VectorParameter() {group=BlockStack Glider|Water, parameterName=Water EmissiveColor, defaultValue=<rgba 0 0.5765 1 1>}
  n28 = SetMaterialAttributes!(
   MaterialAttributes=NamedRerouteUsage()
   Emissive Color=Multiply(A=@n6, B=@n27.RGB)
   Base Color=LinearInterpolate(A=NamedRerouteUsage(), B=@n27.RGB, Alpha=@n6) {constB=0}
   Roughness=@n1
   Specular=ScalarParameter() {defaultValue=0.5, group=BlockStack Glider|Water, parameterName=WaterSpecular}
   Normal=NamedRerouteUsage()
  ) {attributeSetTypes [
    #1
    #2
    #3
    #5
    #4
   ]}
  NamedRerouteDeclaration(None=@n28) {name=AnimatedWater, nodeColor=<rgba 1 0 0.3984 1>}
  n29 = Multiply(A=NamedRerouteUsage(), B=ScalarParameter() {defaultValue=7, group=BlockStack Glider|Cloud, parameterName=Cloud Noise Tiling}) {constB=10}
  n30 = LinearInterpolate(Alpha=TextureSample(UVs=Panner(Coordinate=@n29) {speedX=0.1, speedY=-0.25}) {samplerType=SAMPLERTYPE_Grayscale, texture=&T_Blur_Noise}.R) {constA=-0.2}
  n31 = ScalarParameter() {defaultValue=4, group=BlockStack Glider|Cloud, parameterName=Cloud Noise Intensity}
  NamedRerouteDeclaration(None=SetMaterialAttributes!(MaterialAttributes=NamedRerouteUsage(), World Position Offset=Multiply(A=@n30, B=@n31)) {attributeSetTypes [#6]}) {name=Clouds, nodeColor=<rgba 0.4062 0 1 1>}
  NamedRerouteDeclaration(None=@n14.Base Color) {name=BaseColor In, nodeColor=<rgba 0.1719 1 0 1>}
  NamedRerouteDeclaration(None=@n14.Roughness) {name=Roughness In, nodeColor=<rgba 1 0 0.4922 1>}
 ]
}
